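# Editor utility widget for bulk-editing, selecting and placing light actors in a level.
# Property views bound to the widget's variables feed Apply buttons that set point/spot component or light Blueprint properties on selected actors.
# Selection tools match by light Blueprint, class or detail mode; snapping line-traces to floor, ceiling or wall (local space for walls).
# Auto-attach parents BuildingProps to the nearest BuildingFloor/BuildingWall found by scaled-bounds box overlaps.
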EditorUtilityWidgetBlueprint /WildEstate/ArtTools/EUW_LightActorInspectionUtilities
graphs [
 ApplySelectedLightsAttenuationRadius {nodes [
   ApplySelectedLightsAttenuationRadius() {is=title K2Node_FunctionEntry}
   > n1 = GetSelectedLevelActors(self=Get EditorActorSubsystem())
   > n2 = For Each Loop(Array:Array of Actor Object References=@n1) -> (Array Element:Actor Object Reference)
   > SetEditorProperty(
    Object=GetComponentByClass!(self:Actor Object Reference=@n2, ComponentClass:Actor Component Class Reference=/Script/Engine.PointLightComponent) -> (ReturnValue:Point Light Component Object Reference)
    PropertyName=AttenuationRadius
    PropertyValue:double=Get LightAttenuationRadius() -> (LightAttenuationRadius:double)
   )
  ]}
 ApplySelectedLightsCastShadows {nodes [
   ApplySelectedLightsCastShadows() {is=title K2Node_FunctionEntry}
   > n1 = GetSelectedLevelActors(self=Get EditorActorSubsystem())
   > n2 = For Each Loop(Array=@n1)
   > SetEditorProperty(
    Object=GetComponentByClass(self=@n2, ComponentClass=&PointLightComponent)
    PropertyName=CastShadows
    PropertyValue:bool=Get LightCastShadows() -> (LightCastShadows:bool)
   )
  ]}
 ApplySelectedLightsColor {nodes [
   ApplySelectedLightsColor() {is=title K2Node_FunctionEntry}
   > n1 = GetSelectedLevelActors(self=Get EditorActorSubsystem())
   > n2 = For Each Loop(Array=@n1)
   > SetEditorProperty(
    Object=GetComponentByClass(self=@n2, ComponentClass=&PointLightComponent)
    PropertyName=LightColor
    PropertyValue:Color Structure=Get LightColor() -> (LightColor:Color Structure)
   )
  ]}
 ApplySelectedLightsConeAngle {nodes [
   ApplySelectedLightsConeAngle() {is=title K2Node_FunctionEntry}
   > n1 = GetSelectedLevelActors(self=Get EditorActorSubsystem())
   > n2 = For Each Loop(Array=@n1)
   n3 = GetComponentByClass(self=@n2, ComponentClass=/Script/Engine.SpotLightComponent) -> (ReturnValue:Spot Light Component Object Reference)
   > SetEditorProperty(Object=@n3, PropertyName=OuterConeAngle, PropertyValue:double=Get LightConeAngle() -> (LightConeAngle:double))
   > SetEditorProperty(
    Object=@n3
    PropertyName=InnerConeAngle
    PropertyValue=float - float(A=Get OuterConeAngle!(self:Spot Light Component Object Reference=@n3) -> (OuterConeAngle:float), B=0.5)
   )
  ]}
 ApplySelectedLightsDetailMode {nodes [
   ApplySelectedLightsDetailMode() {is=title K2Node_FunctionEntry}
   > n1 = GetSelectedLevelActors(self=Get EditorActorSubsystem())
   > n2 = For Each Loop(Array=@n1)
   > SetEditorProperty(
    Object=GetComponentByClass(self=@n2, ComponentClass=&PointLightComponent) -> (ReturnValue:Point Light Component Object Reference)
    PropertyName=DetailMode
    PropertyValue:EDetailMode Enum=Get LightDetailMode() -> (LightDetailMode:EDetailMode Enum=DM_Low)
   )
  ]}
 ApplySelectedLightsFalloffExponent {nodes [
   ApplySelectedLightsFalloffExponent() {is=title K2Node_FunctionEntry}
   > n1 = GetSelectedLevelActors(self=Get EditorActorSubsystem())
   > n2 = For Each Loop(Array=@n1)
   > SetEditorProperty(
    Object=GetComponentByClass(self=@n2, ComponentClass=&PointLightComponent)
    PropertyName=LightFalloffExponent
    PropertyValue:double=Get LightFalloffExponent() -> (LightFalloffExponent:double)
   )
  ]}
 ApplySelectedLightsIntensity {nodes [
   ApplySelectedLightsIntensity() {is=title K2Node_FunctionEntry}
   > n1 = GetSelectedLevelActors(self=Get EditorActorSubsystem())
   > n2 = For Each Loop(Array=@n1)
   > SetEditorProperty(
    Object=GetComponentByClass(self=@n2, ComponentClass=&PointLightComponent)
    PropertyName=Intensity
    PropertyValue=Get LightIntensity() -> (LightIntensity:double)
   )
  ]}
 ApplySelectedLightsTimeOfDay {nodes [
   ApplySelectedLightsTimeOfDay() {is=title K2Node_FunctionEntry}
   > n1 = GetSelectedLevelActors(self=Get EditorActorSubsystem())
   > n2 = For Each Loop(Array=@n1)
   > n3 = Cast To BP_Lightsource_CH5_GlobalProp_Parent!(execute, Object=@n2) -> (then, CastFailed, AsBP Lightsource CH5 Global Prop Parent:BP Lightsource CH5 Global Prop Parent Object Reference) {is=title K2Node_DynamicCast}
   > SetEditorProperty(Object=@n3, PropertyName=TimeOfDayLight, PropertyValue:bool=Get LightTimeOfDay() -> (LightTimeOfDay:bool))
  ]}
 AutoAttachLights {nodes [
   AutoAttachLights() {is=title K2Node_FunctionEntry}
   > n1 = GetSelectedLevelActors(self=Get EditorActorSubsystem())
   > n2 = For Each Loop(Array=@n1)
   > n3 = Cast To BuildingProp!(execute, Object=@n2) -> (then, CastFailed, AsBuilding Prop:Building Prop Object Reference) {is=title K2Node_DynamicCast}
   > Set CurrentProp(CurrentProp:Building Prop Object Reference=@n3) -> (Output_Get:Building Prop Object Reference)
   > Clear!(
    execute
    TargetMap:Map of Vectors to Actor Object References=Get AttachmentActorCandidatesMap() -> (AttachmentActorCandidatesMap:Map of Vectors to Actor Object References)
   ) -> (then) {is=title K2Node_CallFunction}
   > Clear(TargetArray:Array of Actor Object References=Get AttachmentActorCandidates() -> (AttachmentActorCandidates:Array of Actor Object References)) {is=title K2Node_CallArrayFunction}
   > n4 = Sequence()
   n5 = Make Array("[0]":EObjectTypeQuery Enum=ObjectTypeQuery1) -> (Array:Array of EObjectTypeQuery Enums)
   n6 = Get CurrentProp() -> (CurrentProp:Building Prop Object Reference)
   n7 = Make Array("[0]":Actor Object Reference=@n6) -> (Array:Array of Actor Object References)
   n8 = Get Actor Bounds(self=@n6, bOnlyCollidingComponents=false)
   n9 = vector * vector(A=@n8.BoxExtent, B=Get AutoAttachRadius() -> (AutoAttachRadius:double)) -> (ReturnValue=~)
   > n10 = Box Overlap Actors!(
    execute
    WorldContextObject:Object Reference
    BoxPos:Vector=@n8
    BoxExtent:Vector=@n9
    ObjectTypes:Array of EObjectTypeQuery Enums=@n5
    ActorClassFilter:Object Class Reference=&BuildingFloor
    ActorsToIgnore:Array of Actor Object References=@n7
   ) -> (then, OutActors:Array of Actor Object References, ReturnValue:bool)
   n11 = Get AttachmentActorCandidates()
   > Append Array!(execute, TargetArray:Array of Actor Object References=@n11, SourceArray:Array of Actor Object References=@n10) -> (then)
   > n12 = Box Overlap Actors(BoxPos=@n8, BoxExtent=@n9, ObjectTypes=@n5, ActorClassFilter=&BuildingWall, ActorsToIgnore=@n7)
   > Append Array(TargetArray=@n11, SourceArray=@n12)
   n13 = Branch(
    execute=@n4.then_1
    Condition=Is Not Empty!(TargetArray:Array of Actor Object References=Get AttachmentActorCandidates()) -> (ReturnValue:bool)
   )
   > n14 = Sequence()
   > n15 = For Each Loop(Array=Get AttachmentActorCandidates())
   > Add!(
    execute
    TargetMap:Map of Vectors to Actor Object References=Get AttachmentActorCandidatesMap()
    Key:"Vector (by ref)"=Get Actor Bounds(self=@n15, bOnlyCollidingComponents=false)
    Value:"Actor Object Reference (by ref)"=@n15
   ) -> (then) {is=title K2Node_CallFunction}
   n16 = Keys!(execute=@n14.then_1, TargetMap:Map of Vectors to Actor Object References=Get AttachmentActorCandidatesMap()) -> (then, Keys:Array of Vectors)
   > n17 = FindNearestVectorInArray!(execute, self, Array:Array of Vectors=@n16, Origin:Vector=Get Actor Location(self=Get CurrentProp())) -> (then, NearestVector:Vector)
   n18 = Get CurrentProp()
   > TransactObject(Object=@n18)
   n19 = Get AttachmentActorCandidatesMap()
   > SetParentActorToAttachTo!(
    execute
    self:Building Static Mesh Object Reference=@n18
    InParentActorToAttachTo:Building Static Mesh Object Reference=Cast To BuildingSMActor!(Object=Find(TargetMap:Map of Vectors to Actor Object References=@n19, Key:"Vector (by ref)"=@n17) -> (Value:Actor Object Reference)) -> (AsBuilding Static Mesh:Building Static Mesh Object Reference, bSuccess:bool=~) {is=title K2Node_DynamicCast}
   ) -> (then)
   n20 = Get CurrentProp()
   PrintString(
    execute=@n13.else
    InString=Append!(A="WARNING: No Attach parent found for ", B=GetDisplayName(Object=@n20), C:String=". Consider increasing search radius!") -> (ReturnValue)
    TextColor=(R=1,G=0,B=0,A=1)
    Duration=5
   )
   > DrawDebugActorsBounds!(execute, self, Actors:Array of Actor Object References=Make Array("[0]"=Get CurrentProp()), Duration:float=5) -> (then)
  ]}
 DrawDebugActorsBounds {nodes [
   n1 = DrawDebugActorsBounds!() -> (then, Actors:Array of Actor Object References, Duration:float=5) {is=title K2Node_FunctionEntry}
   > n2 = For Each Loop(Array=@n1)
   n3 = Get Actor Bounds(self=@n2, bOnlyCollidingComponents=true)
   > DrawDebugBox!(
    execute
    WorldContextObject:Object Reference=@n2
    Center:Vector=@n3
    Extent:Vector=vector * vector(A=@n3.BoxExtent, B:float=1.5) -> (ReturnValue=~)
    LineColor:Linear Color Structure=(R=1,G=0,B=0,A=1)
    Rotation:Rotator
    Duration:float=Get Duration() -> (Duration:float)
    Thickness:float=20
    DepthPriority:EDrawDebugSceneDepthPriorityGroup Enum=Foreground
   ) -> (then)
  ]}
 EventGraph {nodes [
   "On Clicked (Button_SelectLightBPsOfSameType)"() {is=title K2Node_ComponentBoundEvent}
   > SelectLightBPsOfSameType()
   "On Clicked (Button_SelectAllLightBPs)"() {is=title K2Node_ComponentBoundEvent}
   > SelectAllLightActors()
   "On Clicked (Button_AutoAttachLights)"() {is=title K2Node_ComponentBoundEvent}
   > AutoAttachLights() {is=title K2Node_CallFunction}
   "On Clicked (Button_SnapToFloor)"() {is=title K2Node_ComponentBoundEvent}
   > SnapLightToWall!(execute, self, Axis:Vector="0,0,-1") -> (then)
   "On Clicked (Button_ApplyDetailMode)"() {is=title K2Node_ComponentBoundEvent}
   > ApplySelectedLightsDetailMode() {is=title K2Node_CallFunction}
   "On Clicked (Button_SnapToCeiling)"() {is=title K2Node_ComponentBoundEvent}
   > SnapLightToWall(Axis="0,0,1")
   "On Clicked (Button_ApplyTimeOfDay)"() {is=title K2Node_ComponentBoundEvent}
   > ApplySelectedLightsTimeOfDay() {is=title K2Node_CallFunction}
   "On Clicked (Button_SnapToWall)"() {is=title K2Node_ComponentBoundEvent}
   > SnapLightToWall(Axis="0,-1,0")
   "On Clicked (Button_SelectLightsByDetailMode)"() {is=title K2Node_ComponentBoundEvent}
   > SelectLightsByDetailLevel()
   "On Clicked (Button_ApplyIntensity)"() {is=title K2Node_ComponentBoundEvent}
   > ApplySelectedLightsIntensity() {is=title K2Node_CallFunction}
   "On Clicked (Button_ApplyColor)"() {is=title K2Node_ComponentBoundEvent}
   > ApplySelectedLightsColor() {is=title K2Node_CallFunction}
   "On Clicked (Button_ApplyAttenuationRadius)"() {is=title K2Node_ComponentBoundEvent}
   > ApplySelectedLightsAttenuationRadius() {is=title K2Node_CallFunction}
   "On Clicked (Button_ApplyCastShadows)"() {is=title K2Node_ComponentBoundEvent}
   > ApplySelectedLightsCastShadows() {is=title K2Node_CallFunction}
   "On Clicked (Button_ApplyConeAngle)"() {is=title K2Node_ComponentBoundEvent}
   > ApplySelectedLightsConeAngle() {is=title K2Node_CallFunction}
   "On Clicked (Button_ApplyFalloffExponent)"() {is=title K2Node_ComponentBoundEvent}
   > ApplySelectedLightsFalloffExponent() {is=title K2Node_CallFunction}
   Event Construct()
   > Initialize()
  ]}
 FindNearestVectorInArray {nodes [
   n1 = FindNearestVectorInArray!() -> (then, Array:Array of Vectors, Origin:Vector=~) {is=title K2Node_FunctionEntry}
   > n2 = Branch(Condition=Is Empty!(TargetArray:Array of Vectors=@n1) -> (ReturnValue:bool))
   > Return Node!(execute, NearestVector:Vector="0, 0, 0")
   n3 = Sequence!(execute=@n2.else) -> (then_0, then_1, then_2:Exec)
   n4 = Get Array() -> (Array:Array of Vectors)
   > n5 = Set NearestArrayVector(NearestArrayVector:Vector="Get (a copy)"(Array:Array of Vectors=@n4) -> (Output:Vector)) -> (Output_Get:Vector)
   > Set NearestDistanceSquared(NearestDistanceSquared:double="Distance Squared (Vector)"!(V1:Vector=@n5, V2:Vector=Get Origin() -> (Origin:Vector)) -> (ReturnValue:double)) -> (Output_Get:double)
   n6 = For Each Loop(Exec=@n3.then_1, Array:Array of Vectors=@n4) -> (Array Element:Vector)
   n7 = "Distance Squared (Vector)"(V1=@n6, V2=Get Origin())
   > Branch(Condition="float < float"(A=@n7, B=Get NearestDistanceSquared() -> (NearestDistanceSquared:double)))
   > Set NearestDistanceSquared(NearestDistanceSquared=@n7)
   > Set NearestArrayVector(NearestArrayVector=@n6)
   Return Node(execute=@n3.then_2, NearestVector=Get NearestArrayVector() -> (NearestArrayVector:Vector))
  ]}
 Initialize {nodes [
   Initialize() {is=title K2Node_FunctionEntry}
   > SetObject(
    self=[
     Get PropertyView_SelectionDetailMode() -> (PropertyView_SelectionDetailMode:Single Property View Object Reference)
     Get PropertyView_EditDetailMode() -> (PropertyView_EditDetailMode:Single Property View Object Reference)
     Get PropertyView_LightConeAngle() -> (PropertyView_LightConeAngle:Single Property View Object Reference)
     Get PropertyView_AutoAttachRadius() -> (PropertyView_AutoAttachRadius:Single Property View Object Reference)
     Get PropertyView_LightColor() -> (PropertyView_LightColor:Single Property View Object Reference)
     Get PropertyView_LightIntensity() -> (PropertyView_LightIntensity:Single Property View Object Reference)
     Get PropertyView_LightAttenuationRadius() -> (PropertyView_LightAttenuationRadius:Single Property View Object Reference)
     Get PropertyView_LightCastShadows() -> (PropertyView_LightCastShadows:Single Property View Object Reference)
     Get PropertyView_TimeOfDay() -> (PropertyView_TimeOfDay:Single Property View Object Reference)
     Get PropertyView_LightFalloffExponent() -> (PropertyView_LightFalloffExponent:Single Property View Object Reference)
    ]
    NewObject=Self-Reference()
   )
  ]}
 SelectAllLightActors {nodes [
   SelectAllLightActors() {is=title K2Node_FunctionEntry}
   > n1 = Sequence!(execute) -> (then_0, then_1)
   > n2 = GetAllActorsOfClass!(
    execute
    WorldContextObject:Object Reference
    ActorClass:Actor Class Reference=/Game/Environments/Helios/Props/LightSource/BP_Lightsource_CH5_GlobalProp_Parent.BP_Lightsource_CH5_GlobalProp_Parent_C
   ) -> (then, OutActors:Array of BP Lightsource CH5 Global Prop Parent Object References)
   > n3 = For Each Loop(Array:Array of Actor Object References=@n2) -> (Array Element:Actor Object Reference)
   > Is Valid(InputObject=GetComponentByClass(self=@n3, ComponentClass=/Script/Engine.LightComponent) -> (ReturnValue:Light Component Object Reference))
   > Add!(
    execute
    TargetSet:Set of Actor Object References=Get ActorsToSelect() -> (ActorsToSelect:Set of Actor Object References)
    NewItem:"Actor Object Reference (by ref)"=@n3
   ) -> (then)
   n4 = To Array!(execute=@n1.then_1, A:Set of Actor Object References=Get ActorsToSelect()) -> (then, Result:Array of Actor Object References)
   > SetSelectedLevelActors(self=Get EditorActorSubsystem(), ActorsToSelect=@n4)
  ]}
 SelectLightBPsOfSameType {nodes [
   SelectLightBPsOfSameType() {is=title K2Node_FunctionEntry}
   > n1 = Sequence!(execute) -> (then_0, then_1, then_2:Exec)
   > n2 = GetSelectedLevelActors(self=Get EditorActorSubsystem())
   > n3 = For Each Loop(Array=@n2)
   > Is Valid(InputObject=GetComponentByClass(self=@n3, ComponentClass=&LightComponent))
   > Add(
    TargetSet:Set of Actor Class References=Get SelectedLightActorClasses() -> (SelectedLightActorClasses:Set of Actor Class References)
    NewItem:"Actor Class Reference (by ref)"=Get Class(Object=@n3)
   )
   n4 = To Array(execute=@n1.then_2, A=Get ActorsToSelect())
   > SetSelectedLevelActors(self=Get EditorActorSubsystem(), ActorsToSelect=@n4)
   n5 = Branch(
    execute=@n1.then_1
    Condition=Is Not Empty!(TargetSet:Set of Actor Class References=Get SelectedLightActorClasses()) -> (ReturnValue:bool) {is=title K2Node_CallFunction}
   )
   > n6 = To Array(A:Set of Actor Class References=Get SelectedLightActorClasses()) -> (Result:Array of Actor Class References)
   > n7 = For Each Loop(Array:Array of Actor Class References=@n6) -> (Array Element:Actor Class Reference)
   > n8 = GetAllActorsOfClass(ActorClass=@n7) -> (OutActors:Array of Actor Object References)
   > Add Items!(execute, TargetSet:Set of Actor Object References=Get ActorsToSelect(), NewItems:Array of Actor Object References=@n8) -> (then)
   PrintString(execute=@n5.else, InString=WARNING: Selection does not contain any Light BPs, TextColor=(R=1,G=1,B=0,A=1), Duration=5)
  ]}
 SelectLightsByDetailLevel {nodes [
   SelectLightsByDetailLevel() {is=title K2Node_FunctionEntry}
   > n1 = Sequence!(execute) -> (then_0, then_1)
   n2 = Get EditorActorSubsystem()
   > n3 = GetAllLevelActors(self=@n2)
   > n4 = For Each Loop(Array:Array of Actor Object References=@n3) -> (Array Element:Actor Object Reference)
   n5 = GetComponentByClass(self=@n4, ComponentClass=&LightComponent)
   > Is Valid(InputObject=@n5)
   > Branch(Condition="Equal (Enum)"(
     A:EDetailMode Enum=Get DetailMode!(self:Scene Component Object Reference=@n5) -> (DetailMode:EDetailMode Enum=DM_Low)
     B:EDetailMode Enum=Get SelectionDetailMode() -> (SelectionDetailMode:EDetailMode Enum=DM_Low)
    ))
   > Add(TargetSet:Set of Actor Object References=Get ActorsToSelect(), NewItem:"Actor Object Reference (by ref)"=@n4)
   n6 = To Array(execute=@n1.then_1, A:Set of Actor Object References=Get ActorsToSelect()) -> (Result:Array of Actor Object References)
   > SetSelectedLevelActors(self=@n2, ActorsToSelect=@n6)
  ]}
 SnapLightToWall {nodes [
   SnapLightToWall!() -> (then, Axis:Vector="1,0,0") {is=title K2Node_FunctionEntry}
   > n1 = Sequence()
   n2 = GetSelectedLevelActors(execute=@n1.then_1, self=Get EditorActorSubsystem())
   > n3 = For Each Loop(Array=@n2)
   n4 = Get Actor Location(self=@n3)
   n5 = Make Array("[0]"=@n3)
   n6 = Get Axis() -> (Axis:Vector)
   n7 = SelectVector(
    A=TransformDirection!(T:"Transform (by ref)"=Get Actor Transform(self=@n3), Direction:Vector=@n6) -> (ReturnValue:Vector)
    B=@n6
    bPickA=Get TraceLocalSpace() -> (TraceLocalSpace:bool)
   )
   > n8 = Line Trace For Objects!(
    execute
    WorldContextObject:Object Reference
    Start:Vector=@n4
    End:Vector=vector + vector(A=@n4, B=vector * vector(A=@n7, B=10000) -> (ReturnValue=~))
    ObjectTypes:Array of EObjectTypeQuery Enums=Make Array("[0]":EObjectTypeQuery Enum=ObjectTypeQuery1) -> (Array:Array of EObjectTypeQuery Enums)
    bTraceComplex:bool=false
    ActorsToIgnore:Array of Actor Object References=@n5
    DrawDebugType:EDrawDebugTrace Enum
    bIgnoreSelf:bool=true
    TraceColor:Linear Color Structure=(R=1,G=0,B=0,A=1)
    TraceHitColor:Linear Color Structure=(R=0,G=1,B=0,A=1)
    DrawTime:float=5
   ) -> (then, OutHit:Hit Result Structure, ReturnValue:bool)
   > n9 = Branch(Condition=@n8.ReturnValue)
   > TransactObject(Object=@n3)
   > Set Actor Location!(
    execute
    self:Actor Object Reference=@n3
    NewLocation:Vector=BreakHitResult!(Hit:"Hit Result Structure (by ref)"=@n8) -> (
     bBlockingHit:bool
     bInitialOverlap:bool
     Time:float
     Distance:float
     Location:Vector
     ImpactPoint:Vector
     Normal:Vector
     ImpactNormal:Vector
     PhysMat:Physical Material Object Reference
     HitActor:Actor Object Reference
     HitComponent:Primitive Component Object Reference
     HitBoneName:Name
     BoneName:Name
     HitItem:int
     ElementIndex:int
     FaceIndex:int
     TraceStart:Vector
     TraceEnd:Vector
    ).Location
    bSweep:bool=false
    bTeleport:bool=false
   ) -> (then, SweepHitResult:Hit Result Structure, ReturnValue:bool) {is=title K2Node_CallFunction}
   DrawDebugActorsBounds(execute=@n9.else, Actors=@n5, Duration=5) {is=title K2Node_CallFunction}
   Set TraceLocalSpace(
    execute=@n1.then_0
    TraceLocalSpace:bool="Equal (Vector)"!(A:Vector=Get Axis(), B:Vector="0,-1,0", ErrorTolerance:float=0.01) -> (ReturnValue:bool=~) {is=title K2Node_PromotableOperator}
   ) -> (Output_Get:bool)
  ]}
]
variables [
 AutoAttachRadius {type=number, default=2}
 LightAttenuationRadius {type=number, default=250}
 LightCastShadows {type=boolean, default=true}
 LightColor {type=Color, default=<rgba 1 0.9176 0.8549 1>}
 LightConeAngle {type=number, default=35}
 LightDetailMode {type=EDetailMode, default=DM_Medium}
 LightFalloffExponent {type=number, default=0.001}
 LightIntensity {type=number, default=2}
 LightTimeOfDay {type=boolean}
 SelectionDetailMode {type=EDetailMode, default=DM_Medium}
]
widgets {
 info {rootWidgetClass=&EditorUtilityScrollBox, widgetCount=50}
 widgets [
  EditorUtilityScrollBox_0 {widgetClassPath=&EditorUtilityScrollBox, widget=EditorUtilityScrollBox_0}
  Border_0 {
   widgetClassPath=&Border
   parent=EditorUtilityScrollBox_0
   widget=Border_0
   properties {
    visibility=SelfHitTestInvisible
    background {
     bIsDynamicallyLoaded=false
     drawAs=NoDrawType
     imageType=NoImage
     mirroring=NoMirror
     resourceName=None
     resourceObject=null
     tiling=NoTile
     imageSize=<32 32>
     margin=<ltrb 0 0 0 0>
     outlineSettings {
      bUseBrushTransparency=false
      roundingType=HalfHeightRadius
      width=0
      color {colorUseRule=UseColor_Specified, specifiedColor=<rgba 0 0 0 0>}
      cornerRadii=<0 0 0 0>
     }
     tintColor {colorUseRule=UseColor_Specified, specifiedColor=<rgba 1 1 1 1>}
     uVRegion {bIsValid=false, max=<0 0>, min=<0 0>}
    }
    padding=<ltrb 0 2 0 2>
   }
   slot {class=ScrollBoxSlot}
  }
  VerticalBox_1 {widgetClassPath=&VerticalBox, parent=Border_0, widget=VerticalBox_1, slot {class=BorderSlot, properties {padding=<ltrb 0 2 0 2>}}}
  Text_LightBulkEdit {
   widgetClassPath=&TextBlock
   parent=VerticalBox_1
   widget=Text_LightBulkEdit
   properties {
    text=Light Bulk Edit
    visibility=SelfHitTestInvisible
    colorAndOpacity {colorUseRule=UseColor_Specified, specifiedColor=<rgba 1 1 1 0.5>}
    font {
     bForceMonospaced=false
     bMaterialIsStencil=false
     fontMaterial=null
     fontObject=&Roboto
     letterSpacing=0
     monospacedWidth=1
     size=9
     skewAmount=0
     typefaceFontName=Bold
     outlineSettings {
      bApplyOutlineToDropShadows=false
      bMiteredCorners=false
      bSeparateFillAlpha=false
      outlineMaterial=null
      outlineSize=0
      outlineColor=<rgba 0 0 0 1>
     }
    }
   }
   slot {class=VerticalBoxSlot, properties {horizontalAlignment=HAlign_Center, verticalAlignment=VAlign_Center, padding=<ltrb 4 4 4 4>}}
  }
  Border_167 {
   widgetClassPath=&Border
   parent=VerticalBox_1
   widget=Border_167
   properties {
    background {
     bIsDynamicallyLoaded=false
     drawAs=RoundedBox
     imageType=NoImage
     mirroring=NoMirror
     resourceName=None
     resourceObject=null
     tiling=NoTile
     imageSize=<32 32>
     margin=<ltrb 0 0 0 0>
     outlineSettings {
      bUseBrushTransparency=false
      roundingType=FixedRadius
      width=1
      color {colorUseRule=UseColor_Specified, specifiedColor=<rgba 0 0 0 0.5>}
      cornerRadii=<4 4 4 4>
     }
     tintColor {colorUseRule=UseColor_Specified, specifiedColor=<rgba 0 0 0 0>}
     uVRegion {bIsValid=false, max=<0 0>, min=<0 0>}
    }
    padding=<ltrb 4 2 2 2>
   }
   slot {class=VerticalBoxSlot, properties {padding=<ltrb 4 2 4 2>}}
  }
  UniformGridPanel_111 {
   widgetClassPath=/Script/UMG.UniformGridPanel
   parent=Border_167
   widget=UniformGridPanel_111
   properties {slotPadding=<ltrb 2 2 2 2>}
   slot {class=BorderSlot, properties {padding=<ltrb 4 2 2 2>}}
  }
  Button_ApplyColor {
   widgetClassPath=&EditorUtilityButton
   parent=UniformGridPanel_111
   widget=Button_ApplyColor
   slot {class=UniformGridSlot, properties {horizontalAlignment=HAlign_Fill, verticalAlignment=VAlign_Fill}}
  }
  Text_Color {
   widgetClassPath=&TextBlock
   parent=Button_ApplyColor
   widget=Text_Color
   properties {
    text=Apply Color
    visibility=SelfHitTestInvisible
    colorAndOpacity {colorUseRule=UseColor_Specified, specifiedColor=<rgba 1 1 1 0.8>}
    font {
     bForceMonospaced=false
     bMaterialIsStencil=false
     fontMaterial=null
     fontObject=&Roboto
     letterSpacing=0
     monospacedWidth=1
     size=8.25
     skewAmount=0
     typefaceFontName=Bold
     outlineSettings {
      bApplyOutlineToDropShadows=false
      bMiteredCorners=false
      bSeparateFillAlpha=false
      outlineMaterial=null
      outlineSize=0
      outlineColor=<rgba 0 0 0 1>
     }
    }
   }
   slot {class=ButtonSlot, properties {padding=<ltrb 4 4 4 4>}}
  }
  PropertyView_LightColor {
   widgetClassPath=&SinglePropertyView
   parent=UniformGridPanel_111
   widget=PropertyView_LightColor
   properties {nameOverride=" ", propertyName=LightColor}
   slot {class=UniformGridSlot, properties {column=1, horizontalAlignment=HAlign_Fill, verticalAlignment=VAlign_Center}}
  }
  Button_ApplyIntensity {
   widgetClassPath=&EditorUtilityButton
   parent=UniformGridPanel_111
   widget=Button_ApplyIntensity
   slot {class=UniformGridSlot, properties {horizontalAlignment=HAlign_Fill, row=1, verticalAlignment=VAlign_Fill}}
  }
  Text_ApplyIntensity {
   widgetClassPath=&TextBlock
   parent=Button_ApplyIntensity
   widget=Text_ApplyIntensity
   properties {
    text=Apply Intensity
    visibility=SelfHitTestInvisible
    colorAndOpacity {colorUseRule=UseColor_Specified, specifiedColor=<rgba 1 1 1 0.8>}
    font {
     bForceMonospaced=false
     bMaterialIsStencil=false
     fontMaterial=null
     fontObject=&Roboto
     letterSpacing=0
     monospacedWidth=1
     size=8.25
     skewAmount=0
     typefaceFontName=Bold
     outlineSettings {
      bApplyOutlineToDropShadows=false
      bMiteredCorners=false
      bSeparateFillAlpha=false
      outlineMaterial=null
      outlineSize=0
      outlineColor=<rgba 0 0 0 1>
     }
    }
   }
   slot {class=ButtonSlot}
  }
  PropertyView_LightIntensity {
   widgetClassPath=&SinglePropertyView
   parent=UniformGridPanel_111
   widget=PropertyView_LightIntensity
   properties {nameOverride=" ", propertyName=LightIntensity}
   slot {class=UniformGridSlot, properties {column=1, horizontalAlignment=HAlign_Fill, row=1, verticalAlignment=VAlign_Center}}
  }
  Button_ApplyAttenuationRadius {
   widgetClassPath=&EditorUtilityButton
   parent=UniformGridPanel_111
   widget=Button_ApplyAttenuationRadius
   slot {class=UniformGridSlot, properties {horizontalAlignment=HAlign_Fill, row=2, verticalAlignment=VAlign_Fill}}
  }
  Text_ApplyAttenuationRadius {
   widgetClassPath=&TextBlock
   parent=Button_ApplyAttenuationRadius
   widget=Text_ApplyAttenuationRadius
   properties {
    text=Apply Attenuation Radius
    visibility=SelfHitTestInvisible
    colorAndOpacity {colorUseRule=UseColor_Specified, specifiedColor=<rgba 1 1 1 0.8>}
    font {
     bForceMonospaced=false
     bMaterialIsStencil=false
     fontMaterial=null
     fontObject=&Roboto
     letterSpacing=0
     monospacedWidth=1
     size=8.25
     skewAmount=0
     typefaceFontName=Bold
     outlineSettings {
      bApplyOutlineToDropShadows=false
      bMiteredCorners=false
      bSeparateFillAlpha=false
      outlineMaterial=null
      outlineSize=0
      outlineColor=<rgba 0 0 0 1>
     }
    }
   }
   slot {class=ButtonSlot}
  }
  PropertyView_LightAttenuationRadius {
   widgetClassPath=&SinglePropertyView
   parent=UniformGridPanel_111
   widget=PropertyView_LightAttenuationRadius
   properties {nameOverride=" ", propertyName=LightAttenuationRadius}
   slot {class=UniformGridSlot, properties {column=1, horizontalAlignment=HAlign_Fill, row=2, verticalAlignment=VAlign_Center}}
  }
  Button_ApplyConeAngle {
   widgetClassPath=&EditorUtilityButton
   parent=UniformGridPanel_111
   widget=Button_ApplyConeAngle
   slot {class=UniformGridSlot, properties {horizontalAlignment=HAlign_Fill, row=3, verticalAlignment=VAlign_Fill}}
  }
  Text_ApplyConeAngle {
   widgetClassPath=&TextBlock
   parent=Button_ApplyConeAngle
   widget=Text_ApplyConeAngle
   properties {
    text=Apply Cone Angle
    visibility=SelfHitTestInvisible
    colorAndOpacity {colorUseRule=UseColor_Specified, specifiedColor=<rgba 1 1 1 0.8>}
    font {
     bForceMonospaced=false
     bMaterialIsStencil=false
     fontMaterial=null
     fontObject=&Roboto
     letterSpacing=0
     monospacedWidth=1
     size=8.25
     skewAmount=0
     typefaceFontName=Bold
     outlineSettings {
      bApplyOutlineToDropShadows=false
      bMiteredCorners=false
      bSeparateFillAlpha=false
      outlineMaterial=null
      outlineSize=0
      outlineColor=<rgba 0 0 0 1>
     }
    }
   }
   slot {class=ButtonSlot}
  }
  PropertyView_LightConeAngle {
   widgetClassPath=&SinglePropertyView
   parent=UniformGridPanel_111
   widget=PropertyView_LightConeAngle
   properties {nameOverride=" ", propertyName=LightConeAngle}
   slot {class=UniformGridSlot, properties {column=1, horizontalAlignment=HAlign_Fill, row=3, verticalAlignment=VAlign_Center}}
  }
  Button_ApplyFalloffExponent {
   widgetClassPath=&EditorUtilityButton
   parent=UniformGridPanel_111
   widget=Button_ApplyFalloffExponent
   slot {class=UniformGridSlot, properties {horizontalAlignment=HAlign_Fill, row=4, verticalAlignment=VAlign_Fill}}
  }
  Text_ApplyFalloffExponent {
   widgetClassPath=&TextBlock
   parent=Button_ApplyFalloffExponent
   widget=Text_ApplyFalloffExponent
   properties {
    text=Apply Falloff Exponent
    visibility=SelfHitTestInvisible
    colorAndOpacity {colorUseRule=UseColor_Specified, specifiedColor=<rgba 1 1 1 0.8>}
    font {
     bForceMonospaced=false
     bMaterialIsStencil=false
     fontMaterial=null
     fontObject=&Roboto
     letterSpacing=0
     monospacedWidth=1
     size=8.25
     skewAmount=0
     typefaceFontName=Bold
     outlineSettings {
      bApplyOutlineToDropShadows=false
      bMiteredCorners=false
      bSeparateFillAlpha=false
      outlineMaterial=null
      outlineSize=0
      outlineColor=<rgba 0 0 0 1>
     }
    }
   }
   slot {class=ButtonSlot}
  }
  PropertyView_LightFalloffExponent {
   widgetClassPath=&SinglePropertyView
   parent=UniformGridPanel_111
   widget=PropertyView_LightFalloffExponent
   properties {nameOverride=" ", propertyName=LightFalloffExponent}
   slot {class=UniformGridSlot, properties {column=1, horizontalAlignment=HAlign_Fill, row=4, verticalAlignment=VAlign_Center}}
  }
  Button_ApplyCastShadows {
   widgetClassPath=&EditorUtilityButton
   parent=UniformGridPanel_111
   widget=Button_ApplyCastShadows
   slot {class=UniformGridSlot, properties {horizontalAlignment=HAlign_Fill, row=5, verticalAlignment=VAlign_Fill}}
  }
  Text_ApplyCastShadows {
   widgetClassPath=&TextBlock
   parent=Button_ApplyCastShadows
   widget=Text_ApplyCastShadows
   properties {
    text=Apply Cast Shadows
    visibility=SelfHitTestInvisible
    colorAndOpacity {colorUseRule=UseColor_Specified, specifiedColor=<rgba 1 1 1 0.8>}
    font {
     bForceMonospaced=false
     bMaterialIsStencil=false
     fontMaterial=null
     fontObject=&Roboto
     letterSpacing=0
     monospacedWidth=1
     size=8.25
     skewAmount=0
     typefaceFontName=Bold
     outlineSettings {
      bApplyOutlineToDropShadows=false
      bMiteredCorners=false
      bSeparateFillAlpha=false
      outlineMaterial=null
      outlineSize=0
      outlineColor=<rgba 0 0 0 1>
     }
    }
   }
   slot {class=ButtonSlot}
  }
  PropertyView_LightCastShadows {
   widgetClassPath=&SinglePropertyView
   parent=UniformGridPanel_111
   widget=PropertyView_LightCastShadows
   properties {nameOverride=" ", propertyName=LightCastShadows}
   slot {class=UniformGridSlot, properties {column=1, horizontalAlignment=HAlign_Fill, row=5, verticalAlignment=VAlign_Center}}
  }
  Button_ApplyTimeOfDay {
   widgetClassPath=&EditorUtilityButton
   parent=UniformGridPanel_111
   widget=Button_ApplyTimeOfDay
   slot {class=UniformGridSlot, properties {horizontalAlignment=HAlign_Fill, row=6, verticalAlignment=VAlign_Fill}}
  }
  Text_ApplyCastShadows_1 {
   widgetClassPath=&TextBlock
   parent=Button_ApplyTimeOfDay
   widget=Text_ApplyCastShadows_1
   properties {
    text=Apply Time of Day
    visibility=SelfHitTestInvisible
    colorAndOpacity {colorUseRule=UseColor_Specified, specifiedColor=<rgba 1 1 1 0.8>}
    font {
     bForceMonospaced=false
     bMaterialIsStencil=false
     fontMaterial=null
     fontObject=&Roboto
     letterSpacing=0
     monospacedWidth=1
     size=8.25
     skewAmount=0
     typefaceFontName=Bold
     outlineSettings {
      bApplyOutlineToDropShadows=false
      bMiteredCorners=false
      bSeparateFillAlpha=false
      outlineMaterial=null
      outlineSize=0
      outlineColor=<rgba 0 0 0 1>
     }
    }
   }
   slot {class=ButtonSlot}
  }
  PropertyView_TimeOfDay {
   widgetClassPath=&SinglePropertyView
   parent=UniformGridPanel_111
   widget=PropertyView_TimeOfDay
   properties {nameOverride=" ", propertyName=LightTimeOfDay}
   slot {class=UniformGridSlot, properties {column=1, horizontalAlignment=HAlign_Fill, row=6, verticalAlignment=VAlign_Center}}
  }
  Button_ApplyDetailMode {
   widgetClassPath=&EditorUtilityButton
   parent=UniformGridPanel_111
   widget=Button_ApplyDetailMode
   slot {class=UniformGridSlot, properties {horizontalAlignment=HAlign_Fill, row=7, verticalAlignment=VAlign_Fill}}
  }
  Text_ApplyDetailMode {
   widgetClassPath=&TextBlock
   parent=Button_ApplyDetailMode
   widget=Text_ApplyDetailMode
   properties {
    text=Apply Detail Mode
    visibility=SelfHitTestInvisible
    colorAndOpacity {colorUseRule=UseColor_Specified, specifiedColor=<rgba 1 1 1 0.8>}
    font {
     bForceMonospaced=false
     bMaterialIsStencil=false
     fontMaterial=null
     fontObject=&Roboto
     letterSpacing=0
     monospacedWidth=1
     size=8.25
     skewAmount=0
     typefaceFontName=Bold
     outlineSettings {
      bApplyOutlineToDropShadows=false
      bMiteredCorners=false
      bSeparateFillAlpha=false
      outlineMaterial=null
      outlineSize=0
      outlineColor=<rgba 0 0 0 1>
     }
    }
   }
   slot {class=ButtonSlot}
  }
  PropertyView_EditDetailMode {
   widgetClassPath=&SinglePropertyView
   parent=UniformGridPanel_111
   widget=PropertyView_EditDetailMode
   properties {nameOverride=" ", propertyName=LightDetailMode}
   slot {class=UniformGridSlot, properties {column=1, horizontalAlignment=HAlign_Fill, row=7, verticalAlignment=VAlign_Center}}
  }
  Button_AutoAttachLights {
   widgetClassPath=&EditorUtilityButton
   parent=UniformGridPanel_111
   widget=Button_AutoAttachLights
   slot {class=UniformGridSlot, properties {horizontalAlignment=HAlign_Fill, row=8, verticalAlignment=VAlign_Fill}}
  }
  Text_AutoAttachLights {
   widgetClassPath=&TextBlock
   parent=Button_AutoAttachLights
   widget=Text_AutoAttachLights
   properties {
    text=Auto-Attach Lights
    visibility=SelfHitTestInvisible
    colorAndOpacity {colorUseRule=UseColor_Specified, specifiedColor=<rgba 1 1 1 0.8>}
    font {
     bForceMonospaced=false
     bMaterialIsStencil=false
     fontMaterial=null
     fontObject=&Roboto
     letterSpacing=0
     monospacedWidth=1
     size=8.25
     skewAmount=0
     typefaceFontName=Bold
     outlineSettings {
      bApplyOutlineToDropShadows=false
      bMiteredCorners=false
      bSeparateFillAlpha=false
      outlineMaterial=null
      outlineSize=0
      outlineColor=<rgba 0 0 0 1>
     }
    }
   }
   slot {class=ButtonSlot, properties {padding=<ltrb 4 4 4 4>}}
  }
  PropertyView_AutoAttachRadius {
   widgetClassPath=&SinglePropertyView
   parent=UniformGridPanel_111
   widget=PropertyView_AutoAttachRadius
   properties {nameOverride=" ", propertyName=AutoAttachRadius}
   slot {class=UniformGridSlot, properties {column=1, horizontalAlignment=HAlign_Fill, row=8, verticalAlignment=VAlign_Center}}
  }
  Text_LightInspectionTools {
   widgetClassPath=&TextBlock
   parent=VerticalBox_1
   widget=Text_LightInspectionTools
   properties {
    text=Light Utilities
    visibility=SelfHitTestInvisible
    colorAndOpacity {colorUseRule=UseColor_Specified, specifiedColor=<rgba 1 1 1 0.5>}
    font {
     bForceMonospaced=false
     bMaterialIsStencil=false
     fontMaterial=null
     fontObject=&Roboto
     letterSpacing=0
     monospacedWidth=1
     size=9
     skewAmount=0
     typefaceFontName=Bold
     outlineSettings {
      bApplyOutlineToDropShadows=false
      bMiteredCorners=false
      bSeparateFillAlpha=false
      outlineMaterial=null
      outlineSize=0
      outlineColor=<rgba 0 0 0 1>
     }
    }
   }
   slot {class=VerticalBoxSlot, properties {horizontalAlignment=HAlign_Center, verticalAlignment=VAlign_Center, padding=<ltrb 4 4 4 4>}}
  }
  VerticalBox_0 {widgetClassPath=&VerticalBox, parent=VerticalBox_1, widget=VerticalBox_0, slot {class=VerticalBoxSlot}}
  UniformGridPanel_0 {
   widgetClassPath=&UniformGridPanel
   parent=VerticalBox_0
   widget=UniformGridPanel_0
   properties {bOverride_Cursor=true, slotPadding=<ltrb 2 2 2 2>}
   slot {class=VerticalBoxSlot, properties {padding=<ltrb 4 2 4 2>}}
  }
  Button_SelectLightsByDetailMode {
   widgetClassPath=&EditorUtilityButton
   parent=UniformGridPanel_0
   widget=Button_SelectLightsByDetailMode
   slot {class=UniformGridSlot, properties {horizontalAlignment=HAlign_Fill, row=1, verticalAlignment=VAlign_Center}}
  }
  Text_SelectLightsByDetailMode {
   widgetClassPath=&TextBlock
   parent=Button_SelectLightsByDetailMode
   widget=Text_SelectLightsByDetailMode
   properties {
    text=Select Lights by Detail Mode
    visibility=SelfHitTestInvisible
    colorAndOpacity {colorUseRule=UseColor_Specified, specifiedColor=<rgba 1 1 1 0.8>}
    font {
     bForceMonospaced=false
     bMaterialIsStencil=false
     fontMaterial=null
     fontObject=&Roboto
     letterSpacing=0
     monospacedWidth=1
     size=8.25
     skewAmount=0
     typefaceFontName=Bold
     outlineSettings {
      bApplyOutlineToDropShadows=false
      bMiteredCorners=false
      bSeparateFillAlpha=false
      outlineMaterial=null
      outlineSize=0
      outlineColor=<rgba 0 0 0 1>
     }
    }
   }
   slot {class=ButtonSlot, properties {padding=<ltrb 4 4 4 4>}}
  }
  PropertyView_SelectionDetailMode {
   widgetClassPath=&SinglePropertyView
   parent=UniformGridPanel_0
   widget=PropertyView_SelectionDetailMode
   properties {nameOverride=" ", propertyName=SelectionDetailMode}
   slot {class=UniformGridSlot, properties {column=1, horizontalAlignment=HAlign_Fill, row=1, verticalAlignment=VAlign_Center}}
  }
  Button_SelectAllLightBPs {
   widgetClassPath=&EditorUtilityButton
   parent=UniformGridPanel_0
   widget=Button_SelectAllLightBPs
   slot {class=UniformGridSlot, properties {horizontalAlignment=HAlign_Fill, verticalAlignment=VAlign_Fill}}
  }
  Text_SelectAllLightBPs {
   widgetClassPath=&TextBlock
   parent=Button_SelectAllLightBPs
   widget=Text_SelectAllLightBPs
   properties {
    text=Select All Light BPs
    visibility=SelfHitTestInvisible
    colorAndOpacity {colorUseRule=UseColor_Specified, specifiedColor=<rgba 1 1 1 0.8>}
    font {
     bForceMonospaced=false
     bMaterialIsStencil=false
     fontMaterial=null
     fontObject=&Roboto
     letterSpacing=0
     monospacedWidth=1
     size=8.25
     skewAmount=0
     typefaceFontName=Bold
     outlineSettings {
      bApplyOutlineToDropShadows=false
      bMiteredCorners=false
      bSeparateFillAlpha=false
      outlineMaterial=null
      outlineSize=0
      outlineColor=<rgba 0 0 0 1>
     }
    }
   }
   slot {class=ButtonSlot, properties {padding=<ltrb 4 4 4 4>}}
  }
  Button_SelectLightBPsOfSameType {
   widgetClassPath=&EditorUtilityButton
   parent=UniformGridPanel_0
   widget=Button_SelectLightBPsOfSameType
   slot {class=UniformGridSlot, properties {column=1, horizontalAlignment=HAlign_Fill, verticalAlignment=VAlign_Fill}}
  }
  Text_SelectLightBPsOfSameType {
   widgetClassPath=&TextBlock
   parent=Button_SelectLightBPsOfSameType
   widget=Text_SelectLightBPsOfSameType
   properties {
    text=Select Light BPs of Same Type
    visibility=SelfHitTestInvisible
    colorAndOpacity {colorUseRule=UseColor_Specified, specifiedColor=<rgba 1 1 1 0.8>}
    font {
     bForceMonospaced=false
     bMaterialIsStencil=false
     fontMaterial=null
     fontObject=&Roboto
     letterSpacing=0
     monospacedWidth=1
     size=8.25
     skewAmount=0
     typefaceFontName=Bold
     outlineSettings {
      bApplyOutlineToDropShadows=false
      bMiteredCorners=false
      bSeparateFillAlpha=false
      outlineMaterial=null
      outlineSize=0
      outlineColor=<rgba 0 0 0 1>
     }
    }
   }
   slot {class=ButtonSlot, properties {padding=<ltrb 4 4 4 4>}}
  }
  HorizontalBox_194 {
   widgetClassPath=&HorizontalBox
   parent=VerticalBox_0
   widget=HorizontalBox_194
   slot {class=VerticalBoxSlot, properties {padding=<ltrb 4 0 4 0>}}
  }
  Button_SnapToWall {
   widgetClassPath=&EditorUtilityButton
   parent=HorizontalBox_194
   widget=Button_SnapToWall
   slot {class=HorizontalBoxSlot, properties {padding=<ltrb 2 2 2 2>, size {sizeRule=Fill, value=1}}}
  }
  Text_SnapToWall {
   widgetClassPath=&TextBlock
   parent=Button_SnapToWall
   widget=Text_SnapToWall
   properties {
    text=Snap to Wall
    visibility=SelfHitTestInvisible
    colorAndOpacity {colorUseRule=UseColor_Specified, specifiedColor=<rgba 1 1 1 0.8>}
    font {
     bForceMonospaced=false
     bMaterialIsStencil=false
     fontMaterial=null
     fontObject=&Roboto
     letterSpacing=0
     monospacedWidth=1
     size=8.25
     skewAmount=0
     typefaceFontName=Bold
     outlineSettings {
      bApplyOutlineToDropShadows=false
      bMiteredCorners=false
      bSeparateFillAlpha=false
      outlineMaterial=null
      outlineSize=0
      outlineColor=<rgba 0 0 0 1>
     }
    }
   }
   slot {class=ButtonSlot, properties {padding=<ltrb 4 4 4 4>}}
  }
  Button_SnapToFloor {
   widgetClassPath=&EditorUtilityButton
   parent=HorizontalBox_194
   widget=Button_SnapToFloor
   slot {class=HorizontalBoxSlot, properties {padding=<ltrb 2 2 2 2>, size {sizeRule=Fill, value=1}}}
  }
  Text_SnapToFloor {
   widgetClassPath=&TextBlock
   parent=Button_SnapToFloor
   widget=Text_SnapToFloor
   properties {
    text=Snap to Floor
    visibility=SelfHitTestInvisible
    colorAndOpacity {colorUseRule=UseColor_Specified, specifiedColor=<rgba 1 1 1 0.8>}
    font {
     bForceMonospaced=false
     bMaterialIsStencil=false
     fontMaterial=null
     fontObject=&Roboto
     letterSpacing=0
     monospacedWidth=1
     size=8.25
     skewAmount=0
     typefaceFontName=Bold
     outlineSettings {
      bApplyOutlineToDropShadows=false
      bMiteredCorners=false
      bSeparateFillAlpha=false
      outlineMaterial=null
      outlineSize=0
      outlineColor=<rgba 0 0 0 1>
     }
    }
   }
   slot {class=ButtonSlot, properties {padding=<ltrb 4 4 4 4>}}
  }
  Button_SnapToCeiling {
   widgetClassPath=&EditorUtilityButton
   parent=HorizontalBox_194
   widget=Button_SnapToCeiling
   slot {class=HorizontalBoxSlot, properties {padding=<ltrb 2 2 2 2>, size {sizeRule=Fill, value=1}}}
  }
  Text_SnapToCeiling {
   widgetClassPath=&TextBlock
   parent=Button_SnapToCeiling
   widget=Text_SnapToCeiling
   properties {
    text=Snap to Ceiling
    visibility=SelfHitTestInvisible
    colorAndOpacity {colorUseRule=UseColor_Specified, specifiedColor=<rgba 1 1 1 0.8>}
    font {
     bForceMonospaced=false
     bMaterialIsStencil=false
     fontMaterial=null
     fontObject=&Roboto
     letterSpacing=0
     monospacedWidth=1
     size=8.25
     skewAmount=0
     typefaceFontName=Bold
     outlineSettings {
      bApplyOutlineToDropShadows=false
      bMiteredCorners=false
      bSeparateFillAlpha=false
      outlineMaterial=null
      outlineSize=0
      outlineColor=<rgba 0 0 0 1>
     }
    }
   }
   slot {class=ButtonSlot, properties {padding=<ltrb 4 4 4 4>}}
  }
 ]
}
