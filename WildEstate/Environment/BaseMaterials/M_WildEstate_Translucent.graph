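# Unlit, two-sided translucent window glass material for WildEstate, supporting instanced meshes.
# Time-of-day glass and reflection colors combine with fake interior lighting, sun-facing darkening and eye-adaptation compensation for emissive.
# A reflection-line mask, from per-instance, view-offset UVs or the WindowShine function, drives reflection and opacity.
# Actor position seeds per-channel hue shift, then destructible-damage and crack functions process the attributes.
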Material /WildEstate/Environment/BaseMaterials/M_WildEstate_Translucent
graph {
 expressions {
  MaterialExpressionActorPositionWS_1 {class=MaterialExpressionActorPositionWS, inputs [], properties {}}
  MaterialExpressionAdd_0 {
   class=MaterialExpressionAdd
   inputs [
    A {from=MaterialExpressionMaterialFunctionCall_9, output=Color}
    B {from=MaterialExpressionMultiply_4}
   ]
  }
  MaterialExpressionAdd_1 {
   class=MaterialExpressionAdd
   inputs [
    A {from=MaterialExpressionAdd_4}
    B {from=MaterialExpressionMultiply_1}
   ]
  }
  MaterialExpressionAdd_2 {
   class=MaterialExpressionAdd
   inputs [
    A {from=MaterialExpressionMaterialFunctionCall_0, output=R}
    B {from=MaterialExpressionMaterialFunctionCall_0, output=R}
   ]
  }
  MaterialExpressionAdd_3 {
   class=MaterialExpressionAdd
   inputs [
    A {from=MaterialExpressionAdd_2}
    B {from=MaterialExpressionMaterialFunctionCall_0, output=B}
   ]
  }
  MaterialExpressionAdd_4 {
   class=MaterialExpressionAdd
   inputs [
    A {from=MaterialExpressionMultiply_2}
    B {from=MaterialExpressionMultiply_0}
   ]
  }
  MaterialExpressionCameraPositionWS_0 {class=MaterialExpressionCameraPositionWS, inputs [], properties {}}
  MaterialExpressionCollectionParameter_0 {
   class=MaterialExpressionCollectionParameter
   inputs []
   properties {
    collection=&FortniteMaterialParameters
    parameterName=SunlightDirectionalVector
   }
  }
  MaterialExpressionComponentMask_0 {class=MaterialExpressionComponentMask, inputs [None {from=MaterialExpressionTransform_0}], properties {g=true, r=true}}
  MaterialExpressionComponentMask_1 {class=MaterialExpressionComponentMask, inputs [None {from=MaterialExpressionMaterialFunctionCall_5, output=Result}], properties {r=true}}
  MaterialExpressionComponentMask_2 {class=MaterialExpressionComponentMask, inputs [None {from=MaterialExpressionMaterialFunctionCall_6, output=Result}], properties {g=true}}
  MaterialExpressionComponentMask_3 {class=MaterialExpressionComponentMask, inputs [None {from=MaterialExpressionMaterialFunctionCall_7, output=Result}], properties {b=true}}
  MaterialExpressionConstant3Vector_1 {class=MaterialExpressionConstant3Vector}
  MaterialExpressionConvert_0 {
   class=MaterialExpressionConvert
   inputs [
    None {from=MaterialExpressionComponentMask_1}
    None {from=MaterialExpressionComponentMask_2}
    None {from=MaterialExpressionComponentMask_3}
   ]
   properties {
    nodeName=Make Vector3
    convertInputs [
     {
      type=Scalar
      expressionInput {
       expression=&self:MaterialExpressionComponentMask_1
       inputName=None
       mask=0
       maskA=0
       maskB=0
       maskG=0
       maskR=0
       outputIndex=0
      }
     }
     {
      type=Scalar
      expressionInput {
       expression=&self:MaterialExpressionComponentMask_2
       inputName=None
       mask=0
       maskA=0
       maskB=0
       maskG=0
       maskR=0
       outputIndex=0
      }
     }
     {
      type=Scalar
      expressionInput {
       expression=&self:MaterialExpressionComponentMask_3
       inputName=None
       mask=0
       maskA=0
       maskB=0
       maskG=0
       maskR=0
       outputIndex=0
      }
     }
    ]
    convertOutputs [{type=Vector3}]
   }
  }
  MaterialExpressionDivide_0 {
   class=MaterialExpressionDivide
   inputs [
    A {from=MaterialExpressionTransformPosition_0}
    B {from=""}
   ]
   properties {constB=100}
  }
  MaterialExpressionDotProduct_0 {
   class=MaterialExpressionDotProduct
   inputs [
    A {from=MaterialExpressionCollectionParameter_0, output=""}
    B {from=MaterialExpressionMultiply_6, output=""}
   ]
   properties {}
  }
  MaterialExpressionEyeAdaptationInverse_0 {
   class=MaterialExpressionEyeAdaptationInverse
   inputs [
    LightValueInput {from=MaterialExpressionMultiply_5, output=""}
    AlphaInput {from="", output=""}
   ]
   properties {}
  }
  MaterialExpressionFrac_0 {class=MaterialExpressionFrac, inputs [None {from=MaterialExpressionAdd_3, output=""}], properties {}}
  MaterialExpressionFrac_2 {class=MaterialExpressionFrac, inputs [None {from=MaterialExpressionMaterialFunctionCall_8, output=Result}], properties {}}
  MaterialExpressionLength_1 {class=MaterialExpressionLength, inputs [None {from=MaterialExpressionActorPositionWS_1, output=""}], properties {}}
  MaterialExpressionLinearInterpolate_0 {
   class=MaterialExpressionLinearInterpolate
   inputs [
    A {from=MaterialExpressionScalarParameter_3}
    B {from=MaterialExpressionScalarParameter_0}
    Alpha {from=MaterialExpressionNamedRerouteUsage_0}
   ]
  }
  MaterialExpressionMaterialFunctionCall_0 {
   class=MaterialExpressionMaterialFunctionCall
   inputs [Float3 {from=MaterialExpressionDivide_0}]
   properties {materialFunction=&BreakOutFloat3Components}
  }
  MaterialExpressionMaterialFunctionCall_1 {
   class=MaterialExpressionMaterialFunctionCall
   inputs [
    Color_Morning {from=MaterialExpressionVectorParameter_1, output=RGB}
    Color_Day {from=MaterialExpressionVectorParameter_0, output=RGB}
    Color_Evening {from=MaterialExpressionVectorParameter_2, output=RGB}
    Color_Night {from=MaterialExpressionVectorParameter_3, output=RGB}
   ]
   properties {materialFunction=&Color_Time_of_Day}
  }
  MaterialExpressionMaterialFunctionCall_10 {
   class=MaterialExpressionMaterialFunctionCall
   inputs [
    Value_Morning {from=MaterialExpressionVectorParameter_1, output=A}
    Value_Day {from=MaterialExpressionVectorParameter_0, output=A}
    Value_Evening {from=MaterialExpressionVectorParameter_2, output=A}
    Value_Night {from=MaterialExpressionVectorParameter_3, output=A}
   ]
   properties {materialFunction=&Scalar_Time_of_Day}
  }
  MaterialExpressionMaterialFunctionCall_12 {
   class=MaterialExpressionMaterialFunctionCall
   inputs [
    In {from=""}
    In_Translucent {from=MaterialExpressionMaterialFunctionCall_15, output=Material_Out}
    DestructibleDamageAmount {from=""}
   ]
   properties {materialFunction=/WildEstate/Environment/BaseMaterials/MaterialFunctions/MF_WildEstate_Cracks}
  }
  MaterialExpressionMaterialFunctionCall_15 {
   class=MaterialExpressionMaterialFunctionCall
   inputs [
    Material_In {from=MaterialExpressionSetMaterialAttributes_0}
    UseDestructibleDamage {from=""}
    --- usage info tooltip --- {from=""}
    UseCrackedDamageTextures {from=MaterialExpressionStaticBool_3}
    UseDamageWobbleWPO {from=""}
    UseAdditiveWPO {from=MaterialExpressionStaticBool_2}
    WorldPosition {from=""}
    UseDestructionMultiplier {from=""}
    ExposeLerpFunctionForInvertingTheHitDistanceCalc {from=""}
    InvertDamageFalloff {from=""}
    UseAlternateHitGradientOutput {from=""}
    DestructibleDamageAmount {from=""}
    WSDeformationCenter {from=""}
    WSDamageRadius {from=""}
   ]
   properties {materialFunction=&MF_FN_DestructibleDamage}
  }
  MaterialExpressionMaterialFunctionCall_2 {
   class=MaterialExpressionMaterialFunctionCall
   inputs [
    Color_Morning {from=MaterialExpressionVectorParameter_5, output=RGB}
    Color_Day {from=MaterialExpressionVectorParameter_4, output=RGB}
    Color_Evening {from=MaterialExpressionVectorParameter_6, output=RGB}
    Color_Night {from=MaterialExpressionVectorParameter_7, output=RGB}
   ]
   properties {materialFunction=&Color_Time_of_Day}
  }
  MaterialExpressionMaterialFunctionCall_3 {
   class=MaterialExpressionMaterialFunctionCall
   inputs []
   properties {materialFunction=/WildEstate/Environment/BaseMaterials/MaterialFunctions/MF_WildEstate_WindowShine}
  }
  MaterialExpressionMaterialFunctionCall_5 {
   class=MaterialExpressionMaterialFunctionCall
   inputs [
    Hue Shift Percentage {from=MaterialExpressionMultiply_8}
    Texture {from=MaterialExpressionNamedRerouteUsage_4}
   ]
   properties {materialFunction=&HueShift}
  }
  MaterialExpressionMaterialFunctionCall_6 {
   class=MaterialExpressionMaterialFunctionCall
   inputs [
    Hue Shift Percentage {from=MaterialExpressionMultiply_10}
    Texture {from=MaterialExpressionNamedRerouteUsage_4}
   ]
   properties {materialFunction=&HueShift}
  }
  MaterialExpressionMaterialFunctionCall_7 {
   class=MaterialExpressionMaterialFunctionCall
   inputs [
    Hue Shift Percentage {from=MaterialExpressionMultiply_9}
    Texture {from=MaterialExpressionNamedRerouteUsage_4}
   ]
   properties {materialFunction=&HueShift}
  }
  MaterialExpressionMaterialFunctionCall_8 {
   class=MaterialExpressionMaterialFunctionCall
   inputs [
    Base {from=MaterialExpressionLength_1}
    Add {from=MaterialExpressionScalarParameter_10}
   ]
   properties {materialFunction=/Engine/Functions/Engine_MaterialFunctions02/Math/MultiplyAdd}
  }
  MaterialExpressionMaterialFunctionCall_9 {
   class=MaterialExpressionMaterialFunctionCall
   inputs [
    ColorIn {from=MaterialExpressionMaterialFunctionCall_1, output=Result}
    PerPixel Motion Offset {from=""}
    TwoSidedMask {from=""}
    PerInstanceRandom {from=""}
   ]
   properties {materialFunction=/WildEstate/Environment/BaseMaterials/MaterialFunctions/MF_WildEstate_FakeWindowInteriorLighting}
  }
  MaterialExpressionMax_0 {
   class=MaterialExpressionMax
   inputs [
    A {from=MaterialExpressionDotProduct_0}
    B {from=MaterialExpressionScalarParameter_4}
   ]
  }
  MaterialExpressionMultiply_0 {
   class=MaterialExpressionMultiply
   inputs [
    A {from=MaterialExpressionTextureCoordinate_0}
    B {from=MaterialExpressionScalarParameter_1}
   ]
  }
  MaterialExpressionMultiply_1 {
   class=MaterialExpressionMultiply
   inputs [
    A {from=MaterialExpressionComponentMask_0}
    B {from=MaterialExpressionScalarParameter_2}
   ]
  }
  MaterialExpressionMultiply_10 {
   class=MaterialExpressionMultiply
   inputs [
    A {from=MaterialExpressionMultiply_7}
    B {from=MaterialExpressionScalarParameter_8}
   ]
  }
  MaterialExpressionMultiply_12 {
   class=MaterialExpressionMultiply
   inputs [
    A {from=MaterialExpressionMaterialFunctionCall_10, output=Result}
    B {from=MaterialExpressionMaterialFunctionCall_9, output=FakeLightAlphaMulti}
   ]
  }
  MaterialExpressionMultiply_2 {
   class=MaterialExpressionMultiply
   inputs [
    A {from=MaterialExpressionFrac_0}
    B {from=""}
   ]
   properties {constB=33.3}
  }
  MaterialExpressionMultiply_4 {
   class=MaterialExpressionMultiply
   inputs [
    A {from=MaterialExpressionMaterialFunctionCall_2, output=Result}
    B {from=MaterialExpressionOneMinus_0}
   ]
  }
  MaterialExpressionMultiply_5 {
   class=MaterialExpressionMultiply
   inputs [
    A {from=MaterialExpressionAdd_0}
    B {from=MaterialExpressionSaturate_0}
   ]
  }
  MaterialExpressionMultiply_6 {
   class=MaterialExpressionMultiply
   inputs [
    A {from=MaterialExpressionVertexNormalWS_0}
    B {from=MaterialExpressionTwoSidedSign_0}
   ]
  }
  MaterialExpressionMultiply_7 {
   class=MaterialExpressionMultiply
   inputs [
    A {from=MaterialExpressionFrac_2}
    B {from=MaterialExpressionScalarParameter_6}
   ]
  }
  MaterialExpressionMultiply_8 {
   class=MaterialExpressionMultiply
   inputs [
    A {from=MaterialExpressionMultiply_7}
    B {from=MaterialExpressionScalarParameter_7}
   ]
  }
  MaterialExpressionMultiply_9 {
   class=MaterialExpressionMultiply
   inputs [
    A {from=MaterialExpressionMultiply_7}
    B {from=MaterialExpressionScalarParameter_9}
   ]
  }
  MaterialExpressionNamedRerouteDeclaration_0 {
   class=MaterialExpressionNamedRerouteDeclaration
   inputs [None {from=MaterialExpressionLinearInterpolate_0}]
   properties {name=Opacity, nodeColor=<rgba 0.4531 1 0 1>}
  }
  MaterialExpressionNamedRerouteDeclaration_1 {
   class=MaterialExpressionNamedRerouteDeclaration
   inputs [None {from=MaterialExpressionStaticSwitchParameter_0}]
   properties {name=WindowLines, nodeColor=<rgba 1 0 0.7734 1>}
  }
  MaterialExpressionNamedRerouteDeclaration_2 {
   class=MaterialExpressionNamedRerouteDeclaration
   inputs [None {from=MaterialExpressionEyeAdaptationInverse_0, output=EyeAdaptationInverse}]
   properties {name=Emissive, nodeColor=<rgba 0 1 0.9063 1>}
  }
  MaterialExpressionNamedRerouteDeclaration_3 {
   class=MaterialExpressionNamedRerouteDeclaration
   inputs [None {from=MaterialExpressionMultiply_12}]
   properties {name=TODAlpha, nodeColor=<rgba 0 0.8828 1 1>}
  }
  MaterialExpressionNamedRerouteUsage_0 {class=MaterialExpressionNamedRerouteUsage}
  MaterialExpressionNamedRerouteUsage_1 {class=MaterialExpressionNamedRerouteUsage}
  MaterialExpressionNamedRerouteUsage_4 {class=MaterialExpressionNamedRerouteUsage}
  MaterialExpressionNamedRerouteUsage_6 {class=MaterialExpressionNamedRerouteUsage}
  MaterialExpressionNormalize_0 {class=MaterialExpressionNormalize, inputs [VectorInput {from=MaterialExpressionSubtract_0, output=""}]}
  MaterialExpressionOneMinus_0 {class=MaterialExpressionOneMinus, inputs [None {from=MaterialExpressionNamedRerouteUsage_1}]}
  MaterialExpressionSaturate_0 {class=MaterialExpressionSaturate, inputs [None {from=MaterialExpressionMax_0}]}
  MaterialExpressionScalarParameter_0 {class=MaterialExpressionScalarParameter, properties {defaultValue=0.744, parameterName=Opacity_Glass}}
  MaterialExpressionScalarParameter_1 {class=MaterialExpressionScalarParameter, properties {defaultValue=0.5, parameterName=GlassRelectionTiling}}
  MaterialExpressionScalarParameter_10 {
   class=MaterialExpressionScalarParameter
   properties {defaultValue=0.05, group=Random Variation, parameterName=HueShift.Seed, sortPriority=10}
  }
  MaterialExpressionScalarParameter_2 {class=MaterialExpressionScalarParameter, properties {defaultValue=0.1, parameterName=GlassRelectionOffset}}
  MaterialExpressionScalarParameter_3 {class=MaterialExpressionScalarParameter, properties {defaultValue=1, parameterName=Opacity_Lines}}
  MaterialExpressionScalarParameter_4 {class=MaterialExpressionScalarParameter, properties {defaultValue=0.4109, parameterName=MinDarkness}}
  MaterialExpressionScalarParameter_6 {
   class=MaterialExpressionScalarParameter
   properties {
    defaultValue=0.1
    desc="HueShift * Actor Pos Rand (0-1) = Shift Amount"
    group=Random Variation
    parameterName=HueShift.Intensity
    sortPriority=1
   }
  }
  MaterialExpressionScalarParameter_7 {
   class=MaterialExpressionScalarParameter
   properties {
    defaultValue=0.975
    desc=How much the Hue shift Infulences the RED channel
    group=Random Variation
    parameterName=HueShift.MaxVariation.Red
    sortPriority=4
   }
  }
  MaterialExpressionScalarParameter_8 {
   class=MaterialExpressionScalarParameter
   properties {
    defaultValue=0.975
    desc=How much the Hue shift Infulences the GREEN channel
    group=Random Variation
    parameterName=HueShift.MaxVariation.Green
    sortPriority=5
   }
  }
  MaterialExpressionScalarParameter_9 {
   class=MaterialExpressionScalarParameter
   properties {
    defaultValue=0.975
    desc=How much the Hue shift Infulences the BLUE channel
    group=Random Variation
    parameterName=HueShift.MaxVariation.Blue
    sortPriority=6
   }
  }
  MaterialExpressionSetMaterialAttributes_0 {
   class=MaterialExpressionSetMaterialAttributes
   inputs [
    MaterialAttributes {from="", output=""}
    Emissive Color {from=MaterialExpressionConvert_0, output=""}
    Opacity {from=MaterialExpressionNamedRerouteUsage_6, output=""}
   ]
   properties {attributeSetTypes [#1, #2]}
  }
  MaterialExpressionStaticBool_2 {class=MaterialExpressionStaticBool, properties {value=true}}
  MaterialExpressionStaticBool_3 {class=MaterialExpressionStaticBool}
  MaterialExpressionStaticSwitchParameter_0 {
   class=MaterialExpressionStaticSwitchParameter
   inputs [
    True {from=MaterialExpressionMaterialFunctionCall_3, output=Mask}
    False {from=MaterialExpressionTextureSample_0, output=R}
   ]
   properties {
    desc="(Previously named Use Planar Mapping)\r\n\r\nFalse = Default behavior using reflection vector, better for any curved surfaces\r\nTrue - Select between World Aligned, Local Aligned"
    parameterName=UseProjectedUVs
   }
  }
  MaterialExpressionSubtract_0 {
   class=MaterialExpressionSubtract
   inputs [
    A {from=MaterialExpressionTransformPosition_0}
    B {from=MaterialExpressionCameraPositionWS_0}
   ]
  }
  MaterialExpressionTextureCoordinate_0 {class=MaterialExpressionTextureCoordinate}
  MaterialExpressionTextureSample_0 {
   class=MaterialExpressionTextureSample
   inputs [
    UVs {from=MaterialExpressionAdd_1}
    Tex {from=""}
    Apply View MipBias {from=""}
   ]
   properties {texture=/WildEstate/Environment/Textures/T_WildEstate_Windows}
  }
  MaterialExpressionTransformPosition_0 {
   class=MaterialExpressionTransformPosition
   inputs [None {from=MaterialExpressionConstant3Vector_1, output=""}]
   properties {transformSourceType=TRANSFORMPOSSOURCE_Instance, transformType=TRANSFORMPOSSOURCE_World}
  }
  MaterialExpressionTransform_0 {
   class=MaterialExpressionTransform
   inputs [None {from=MaterialExpressionNormalize_0}]
   properties {transformSourceType=TRANSFORMSOURCE_World, transformType=TRANSFORM_Tangent}
  }
  MaterialExpressionTwoSidedSign_0 {class=MaterialExpressionTwoSidedSign, inputs [], properties {}}
  MaterialExpressionVectorParameter_0 {class=MaterialExpressionVectorParameter, properties {parameterName=Glass_Color_Day, defaultValue=<rgba 0.1094 0.7 0.5916 1>}}
  MaterialExpressionVectorParameter_1 {class=MaterialExpressionVectorParameter, properties {parameterName=Glass_Color_Morning, defaultValue=<rgba 0.1747 0.2248 0.55 1>}}
  MaterialExpressionVectorParameter_2 {class=MaterialExpressionVectorParameter, properties {parameterName=Glass_Color_Evening, defaultValue=<rgba 0.125 0.2124 0.5 1>}}
  MaterialExpressionVectorParameter_3 {class=MaterialExpressionVectorParameter, properties {parameterName=Glass_Color_Night, defaultValue=<rgba 0.0198 0.3849 0.4 1>}}
  MaterialExpressionVectorParameter_4 {class=MaterialExpressionVectorParameter, properties {parameterName=Reflection_Color_Day, defaultValue=<rgba 0.1131 0.724 0.6119 1>}}
  MaterialExpressionVectorParameter_5 {class=MaterialExpressionVectorParameter, properties {parameterName=Reflection_Color_Morning, defaultValue=<rgba 0.1964 0.3241 0.4375 1>}}
  MaterialExpressionVectorParameter_6 {class=MaterialExpressionVectorParameter, properties {parameterName=Reflection_Color_Evening, defaultValue=<rgba 0.1964 0.3241 0.4375 1>}}
  MaterialExpressionVectorParameter_7 {class=MaterialExpressionVectorParameter, properties {parameterName=Reflection_Color_Night, defaultValue=<rgba 0.0371 0.258 0.276 1>}}
  MaterialExpressionVertexNormalWS_0 {class=MaterialExpressionVertexNormalWS, inputs [], properties {}}
 }
 properties {
  bAutomaticallySetUsageInEditor=false
  bUseMaterialAttributes=true
  bUsedWithInstancedStaticMeshes=true
  blendMode=BLEND_Translucent
  shadingModel=MSM_Unlit
  twoSided=true
  usedShadingModels=Unlit
 }
}
