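# Editor utility widget for WildEstate level art that adjusts window glass on selected actors, with each button wrapped in an undo transaction.
# ToggleCustomPrim writes FakeInteriorLight or SwapABFaceSide as default custom primitive data on components whose base material is WindowsGlass or Translucent.
# SetTwoSidedMaterial swaps slots between single- and two-sided glass by exact material match.
# TrySetGlass forces slots with either base material to default WindowsGlass.
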EditorUtilityWidgetBlueprint /WildEstate/ArtTools/EUW_WidlEstate_WindowFakeLightToggle
graphs [
 EventGraph {nodes [
   "On Clicked (Button_Off)"() {is=title K2Node_ComponentBoundEvent}
   > Create Transaction!(Begin:Exec, Finalize:Exec=@n1.then, Cancel:Exec, Context:String, Description:Text="LOCTEXT(\"Remove Fake Lighting on Glass\")") -> (Pending:Exec, Complete:Exec) {is=title K2Node_MacroInstance}
   > n1 = ToggleCustomPrim!(execute, self, ParameterName:Name=FakeInteriorLight, Value:float=0) -> (then)
   "On Clicked (Button_On)"() {is=title K2Node_ComponentBoundEvent}
   > Create Transaction(Finalize=@n2.then, Description="LOCTEXT(\"Enable Fake Lighting on Glass\")")
   > n2 = ToggleCustomPrim(ParameterName=FakeInteriorLight, Value=1)
   "On Clicked (Button_SwapSides)"() {is=title K2Node_ComponentBoundEvent}
   > Create Transaction(Finalize=@n3.then, Description="LOCTEXT(\"Swap Lit Side\")")
   > n3 = ToggleCustomPrim(ParameterName=SwapABFaceSide, Value=1)
   "On Clicked (Button_ResetSides)"() {is=title K2Node_ComponentBoundEvent}
   > Create Transaction(Finalize=@n4.then, Description="LOCTEXT(\"Reset Lit Side\")")
   > n4 = ToggleCustomPrim(ParameterName=SwapABFaceSide, Value=0)
   "On Clicked (Button_SetTwoSidedOn)"() {is=title K2Node_ComponentBoundEvent}
   > Create Transaction(Finalize=@n5.then, Description="LOCTEXT(\"Set Glass Two Sided Material\")")
   > n5 = SetTwoSidedMaterial!(execute, self, bUseTwoSided:bool=true) -> (then)
   "On Clicked (Button_SetTwoSidedOff)"() {is=title K2Node_ComponentBoundEvent}
   > Create Transaction(Finalize=@n6.then, Description="LOCTEXT(\"Set Glass Single Sided Material\")")
   > n6 = SetTwoSidedMaterial(bUseTwoSided=false)
   "On Clicked (Button_ForceSetDefaultGlass)"() {is=title K2Node_ComponentBoundEvent}
   > Create Transaction(Finalize=@n7.then, Description="LOCTEXT(\"Set Glass Single Sided Material\")")
   > n7 = TrySetGlass()
   Event PreConstruct()
   Event Construct()
   Event Tick()
  ]}
 SetTwoSidedMaterial {nodes [
   SetTwoSidedMaterial!() -> (then, bUseTwoSided:bool=~) {is=title K2Node_FunctionEntry}
   > n1 = GetSelectedLevelActors(self=Get EditorActorSubsystem())
   > Branch(Condition=Is Not Empty!(TargetArray:Array of Actor Object References=@n1) -> (ReturnValue:bool))
   > n2 = For Each Loop(Array:Array of Actor Object References=@n1) -> (Array Element:Actor Object Reference)
   > Set Actor(Actor:Actor Object Reference=@n2) -> (Output_Get:Actor Object Reference)
   > n3 = For Each Loop(Array:Array of Primitive Component Object References=Get Components By Class!(self:Actor Object Reference=@n2, ComponentClass:Actor Component Class Reference=/Script/Engine.PrimitiveComponent) -> (ReturnValue:Array of Primitive Component Object References)) -> (Array Element:Primitive Component Object Reference)
   > n4 = Set Prim(Prim:Primitive Component Object Reference=@n3) -> (Output_Get:Primitive Component Object Reference)
   > Is Valid(InputObject=@n4)
   n5 = Get Prim() -> (Prim:Primitive Component Object Reference)
   > n6 = For Loop(LastIndex=int - int(A=GetNumMaterials!(self:Primitive Component Object Reference=@n5) -> (ReturnValue:int=0)))
   > Set ElementID(ElementID:int=@n6) -> (Output_Get:int=0)
   n7 = GetMaterial!(self:Primitive Component Object Reference=Get Prim(), ElementIndex:int=@n6) -> (ReturnValue:Material Interface Object Reference)
   > Is Valid(InputObject=@n7)
   n8 = "Equal (Object)"(A=@n7, B=/WildEstate/Environment/BaseMaterials/M_WildEstate_Translucent)
   n9 = "Equal (Object)"(A=@n7, B=/WildEstate/Environment/BaseMaterials/M_WildEstate_WindowsGlass)
   > n10 = Branch(Condition=AND Boolean(A=OR Boolean(A=@n8, B=@n9), B=Get bUseTwoSided() -> (bUseTwoSided:bool)))
   > TransactObject(Object=Get Prim())
   n11 = Get ElementID() -> (ElementID:int=0)
   > SetMaterial(self=Get Prim(), ElementIndex=@n11, Material=/WildEstate/Environment/BaseMaterials/MI_WildEstate_WindowsGlass_TwoSided)
   n12 = Get bUseTwoSided()
   Branch(
    execute=@n10.else
    Condition=AND Boolean(A="Equal (Object)"(A=@n7, B=&MI_WildEstate_WindowsGlass_TwoSided), B=NOT Boolean(A=@n12))
   )
   > TransactObject(Object=Get Prim())
   > SetMaterial(self=Get Prim(), ElementIndex=@n11, Material=&M_WildEstate_WindowsGlass)
  ]}
 ToggleCustomPrim {nodes [
   ToggleCustomPrim!() -> (then, ParameterName:Name, Value:float=~) {is=title K2Node_FunctionEntry}
   > n1 = GetSelectedLevelActors(self=Get EditorActorSubsystem())
   > Branch(Condition=Is Not Empty(TargetArray=@n1))
   > n2 = For Each Loop(Array:Array of Actor Object References=@n1) -> (Array Element:Actor Object Reference)
   > Set Actor(Actor=@n2)
   > n3 = For Each Loop(Array:Array of Primitive Component Object References=Get Components By Class(self=@n2, ComponentClass=&PrimitiveComponent)) -> (Array Element:Primitive Component Object Reference)
   > n4 = Set Prim(Prim=@n3)
   > Is Valid(InputObject=@n4)
   n5 = Get Prim()
   > n6 = For Loop(LastIndex=int - int(A=GetNumMaterials(self=@n5) -> (ReturnValue=0)))
   n7 = GetMaterial(self=Get Prim(), ElementIndex=@n6)
   > Is Valid(InputObject=@n7)
   > n8 = GetBaseMaterial!(execute, self:Material Interface Object Reference=@n7) -> (then, ReturnValue:Material Object Reference)
   > Branch(Condition=OR Boolean(
     A="Equal (Object)"(A=@n8, B=&M_WildEstate_Translucent)
     B="Equal (Object)"(A=@n8, B=&M_WildEstate_WindowsGlass)
    ))
   > TransactObject(Object=Get Prim())
   > SetScalarParameterForDefaultCustomPrimitiveData!(
    execute
    self:Primitive Component Object Reference=Get Prim()
    ParameterName:Name=Get ParameterName() -> (ParameterName:Name=None)
    Value:float=Get Value() -> (Value:float)
   ) -> (then)
  ]}
 TrySetGlass {nodes [
   TrySetGlass() {is=title K2Node_FunctionEntry}
   > n1 = GetSelectedLevelActors(self=Get EditorActorSubsystem())
   > Branch(Condition=Is Not Empty(TargetArray=@n1))
   > n2 = For Each Loop(Array:Array of Actor Object References=@n1) -> (Array Element:Actor Object Reference)
   > Set Actor(Actor=@n2)
   > n3 = For Each Loop(Array:Array of Primitive Component Object References=Get Components By Class(self=@n2, ComponentClass=&PrimitiveComponent)) -> (Array Element:Primitive Component Object Reference)
   > n4 = Set Prim(Prim=@n3)
   > Is Valid(InputObject=@n4)
   n5 = Get Prim()
   > n6 = For Loop(LastIndex=int - int(A=GetNumMaterials(self=@n5) -> (ReturnValue=0)))
   > Set ElementID(ElementID=@n6) -> (Output_Get=0)
   n7 = GetMaterial(self=Get Prim(), ElementIndex=@n6)
   > Is Valid(InputObject=@n7)
   > n8 = GetBaseMaterial(self=@n7)
   > Branch(Condition=OR Boolean(
     A="Equal (Object)"(A=@n8, B=&M_WildEstate_Translucent)
     B="Equal (Object)"(A=@n8, B=&M_WildEstate_WindowsGlass)
    ))
   > TransactObject(Object=Get Prim())
   > SetMaterial(self=Get Prim(), ElementIndex=Get ElementID() -> (ElementID=0), Material=&M_WildEstate_WindowsGlass)
  ]}
]
variables [
 Actor {type=Actor, default=null}
 Prim {type=PrimitiveComponent, default=null}
]
widgets {
 info {rootWidgetClass=&Overlay, widgetCount=22}
 widgets [
  Overlay_0 {widgetClassPath=&Overlay, widget=Overlay_0}
  VerticalBox_1 {widgetClassPath=&VerticalBox, parent=Overlay_0, widget=VerticalBox_1, slot {class=OverlaySlot}}
  Border_0 {
   widgetClassPath=&Border
   parent=VerticalBox_1
   widget=Border_0
   properties {brushColor=<rgba 0.026 0.026 0.026 1>}
   slot {class=VerticalBoxSlot}
  }
  VerticalBox_0 {widgetClassPath=&VerticalBox, parent=Border_0, widget=VerticalBox_0, slot {class=BorderSlot}}
  TextHeader_1 {
   widgetClassPath=&TextBlock
   parent=VerticalBox_0
   widget=TextHeader_1
   properties {
    text=Fake Lighting Custom Prim Controls
    font {
     bForceMonospaced=false
     bMaterialIsStencil=false
     fontMaterial=null
     fontObject=&Roboto
     letterSpacing=0
     monospacedWidth=1
     size=13.5
     skewAmount=0
     typefaceFontName=Bold
     outlineSettings {
      bApplyOutlineToDropShadows=false
      bMiteredCorners=false
      bSeparateFillAlpha=false
      outlineMaterial=null
      outlineSize=0
      outlineColor=<rgba 0 0 0 1>
     }
    }
   }
   slot {class=VerticalBoxSlot, properties {horizontalAlignment=HAlign_Center, padding=<ltrb 8 8 8 8>}}
  }
  Button_On {widgetClassPath=&EditorUtilityButton, parent=VerticalBox_0, widget=Button_On, slot {class=VerticalBoxSlot}}
  TextBlock_54 {
   widgetClassPath=&TextBlock
   parent=Button_On
   widget=TextBlock_54
   properties {
    text=Enable Fake Lighting
    font {
     bForceMonospaced=false
     bMaterialIsStencil=false
     fontMaterial=null
     fontObject=&Roboto
     letterSpacing=0
     monospacedWidth=1
     size=13.5
     skewAmount=0
     typefaceFontName=Regular
     outlineSettings {
      bApplyOutlineToDropShadows=false
      bMiteredCorners=false
      bSeparateFillAlpha=false
      outlineMaterial=null
      outlineSize=0
      outlineColor=<rgba 0 0 0 1>
     }
    }
   }
   slot {class=ButtonSlot}
  }
  Button_Off {widgetClassPath=&EditorUtilityButton, parent=VerticalBox_0, widget=Button_Off, slot {class=VerticalBoxSlot}}
  TextBlock {
   widgetClassPath=&TextBlock
   parent=Button_Off
   widget=TextBlock
   properties {
    text=Remove Fake Lighting
    font {
     bForceMonospaced=false
     bMaterialIsStencil=false
     fontMaterial=null
     fontObject=&Roboto
     letterSpacing=0
     monospacedWidth=1
     size=13.5
     skewAmount=0
     typefaceFontName=Regular
     outlineSettings {
      bApplyOutlineToDropShadows=false
      bMiteredCorners=false
      bSeparateFillAlpha=false
      outlineMaterial=null
      outlineSize=0
      outlineColor=<rgba 0 0 0 1>
     }
    }
   }
   slot {class=ButtonSlot}
  }
  Button_SwapSides {widgetClassPath=&EditorUtilityButton, parent=VerticalBox_0, widget=Button_SwapSides, slot {class=VerticalBoxSlot}}
  TextBlock_1 {
   widgetClassPath=&TextBlock
   parent=Button_SwapSides
   widget=TextBlock_1
   properties {
    text=Swap Sides
    font {
     bForceMonospaced=false
     bMaterialIsStencil=false
     fontMaterial=null
     fontObject=&Roboto
     letterSpacing=0
     monospacedWidth=1
     size=13.5
     skewAmount=0
     typefaceFontName=Regular
     outlineSettings {
      bApplyOutlineToDropShadows=false
      bMiteredCorners=false
      bSeparateFillAlpha=false
      outlineMaterial=null
      outlineSize=0
      outlineColor=<rgba 0 0 0 1>
     }
    }
   }
   slot {class=ButtonSlot}
  }
  Button_ResetSides {widgetClassPath=&EditorUtilityButton, parent=VerticalBox_0, widget=Button_ResetSides, slot {class=VerticalBoxSlot}}
  TextBlock_2 {
   widgetClassPath=&TextBlock
   parent=Button_ResetSides
   widget=TextBlock_2
   properties {
    text=Reset Sides
    font {
     bForceMonospaced=false
     bMaterialIsStencil=false
     fontMaterial=null
     fontObject=&Roboto
     letterSpacing=0
     monospacedWidth=1
     size=13.5
     skewAmount=0
     typefaceFontName=Regular
     outlineSettings {
      bApplyOutlineToDropShadows=false
      bMiteredCorners=false
      bSeparateFillAlpha=false
      outlineMaterial=null
      outlineSize=0
      outlineColor=<rgba 0 0 0 1>
     }
    }
   }
   slot {class=ButtonSlot}
  }
  Border_54 {
   widgetClassPath=&Border
   parent=VerticalBox_1
   widget=Border_54
   properties {brushColor=<rgba 0.026 0.026 0.026 1>}
   slot {class=VerticalBoxSlot, properties {padding=<ltrb 0 24 0 0>}}
  }
  VerticalBox_44 {widgetClassPath=&VerticalBox, parent=Border_54, widget=VerticalBox_44, slot {class=BorderSlot}}
  TextHeader {
   widgetClassPath=&TextBlock
   bIsVariable=true
   parent=VerticalBox_44
   widget=TextHeader
   properties {
    text=Swap Materials
    font {
     bForceMonospaced=false
     bMaterialIsStencil=false
     fontMaterial=null
     fontObject=&Roboto
     letterSpacing=0
     monospacedWidth=1
     size=13.5
     skewAmount=0
     typefaceFontName=Bold
     outlineSettings {
      bApplyOutlineToDropShadows=false
      bMiteredCorners=false
      bSeparateFillAlpha=false
      outlineMaterial=null
      outlineSize=0
      outlineColor=<rgba 0 0 0 1>
     }
    }
   }
   slot {class=VerticalBoxSlot, properties {horizontalAlignment=HAlign_Center, padding=<ltrb 8 8 8 8>}}
  }
  Button_SetTwoSidedOff {widgetClassPath=&EditorUtilityButton, parent=VerticalBox_44, widget=Button_SetTwoSidedOff, slot {class=VerticalBoxSlot}}
  TextBlock_4 {
   widgetClassPath=&TextBlock
   parent=Button_SetTwoSidedOff
   widget=TextBlock_4
   properties {
    text=Use Single Sided Glass Instance
    font {
     bForceMonospaced=false
     bMaterialIsStencil=false
     fontMaterial=null
     fontObject=&Roboto
     letterSpacing=0
     monospacedWidth=1
     size=13.5
     skewAmount=0
     typefaceFontName=Regular
     outlineSettings {
      bApplyOutlineToDropShadows=false
      bMiteredCorners=false
      bSeparateFillAlpha=false
      outlineMaterial=null
      outlineSize=0
      outlineColor=<rgba 0 0 0 1>
     }
    }
   }
   slot {class=ButtonSlot}
  }
  Button_SetTwoSidedOn {widgetClassPath=&EditorUtilityButton, parent=VerticalBox_44, widget=Button_SetTwoSidedOn, slot {class=VerticalBoxSlot}}
  TextBlock_3 {
   widgetClassPath=&TextBlock
   parent=Button_SetTwoSidedOn
   widget=TextBlock_3
   properties {
    text=Use Two Sided Glass Instance
    font {
     bForceMonospaced=false
     bMaterialIsStencil=false
     fontMaterial=null
     fontObject=&Roboto
     letterSpacing=0
     monospacedWidth=1
     size=13.5
     skewAmount=0
     typefaceFontName=Regular
     outlineSettings {
      bApplyOutlineToDropShadows=false
      bMiteredCorners=false
      bSeparateFillAlpha=false
      outlineMaterial=null
      outlineSize=0
      outlineColor=<rgba 0 0 0 1>
     }
    }
   }
   slot {class=ButtonSlot}
  }
  Button_ForceSetDefaultGlass {
   widgetClassPath=&EditorUtilityButton
   parent=VerticalBox_44
   widget=Button_ForceSetDefaultGlass
   properties {toolTipText=Attempts to set the Glass to be the Default Glass}
   slot {class=VerticalBoxSlot}
  }
  TextBlock_5 {
   widgetClassPath=&TextBlock
   parent=Button_ForceSetDefaultGlass
   widget=TextBlock_5
   properties {
    text=Force Set Default Glass
    font {
     bForceMonospaced=false
     bMaterialIsStencil=false
     fontMaterial=null
     fontObject=&Roboto
     letterSpacing=0
     monospacedWidth=1
     size=13.5
     skewAmount=0
     typefaceFontName=Regular
     outlineSettings {
      bApplyOutlineToDropShadows=false
      bMiteredCorners=false
      bSeparateFillAlpha=false
      outlineMaterial=null
      outlineSize=0
      outlineColor=<rgba 0 0 0 1>
     }
    }
   }
   slot {class=ButtonSlot}
  }
 ]
}
