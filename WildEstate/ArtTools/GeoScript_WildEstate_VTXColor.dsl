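; Geometry Script editor tool actor that paints per-piece vertex colours on a static mesh's LODs.
; On rebuild, it splits each LOD into angle-based polygroups and assigns each an index-based grey.
; Where UV channel 1 exists, pieces whose first rounded-up U misses the enabled Layer1-4 numbers turn white instead.
; Welded pieces are recombined and previewed beside the originals, and Bake writes them back to the source asset.
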(event Events|EventOnRebuildGeneratedMesh
  (params (TargetMesh :type Dynamic%20Mesh%20Object%20Reference))
  (bind _returnvalue 1.0)
  (bind _layer4 (Variables|Default|GetLayer4))
  (bind _returnvalue_2 (Math|Conversions|ToInteger(Boolean) _layer4))
  (bind _layer1 (Variables|Default|GetLayer1))
  (bind _returnvalue_3 (Math|Conversions|ToInteger(Boolean) _layer1))
  (bind _layer2 (Variables|Default|GetLayer2))
  (bind _returnvalue_4 (Math|Conversions|ToInteger(Boolean) _layer2))
  (bind _layer3 (Variables|Default|GetLayer3))
  (bind _returnvalue_5 (Math|Conversions|ToInteger(Boolean) _layer3))
  (bind _staticmesh (Variables|Default|GetStaticmesh))
  (bind _returnvalue_6 (StaticMesh|GetNumLODs _staticmesh))
  (bind _returnvalue_7 (* _returnvalue_2 4))
  (bind _returnvalue_8 (* _returnvalue_4 2))
  (bind _returnvalue_9 (* _returnvalue_5 3))
  (Variables|Default|SetTarget%20Mesh TargetMesh)
  (Development|PrintString (Utilities|String|Append "Number of Lods to Process " (Utilities|String|ToString(Integer) _returnvalue_6)))
  (for _index (range (+ _returnvalue_6 1))
    (bind _returnvalue_10 (DynamicMeshActor|AllocateComputeMesh))
    (Utilities|Array|Add (Variables|Default|GetTarget%20Mesh_LODs) _returnvalue_10)
    (bind _output_get (Variables|Default|SetCurrent%20Mesh%20LOD _returnvalue_10))
    (bind _returnvalue_11 (GeometryScript|StaticMesh|CopyMeshFromStaticMesh :FromStaticMeshAsset _staticmesh :ToDynamicMesh _output_get :RequestedLOD (select (== _index 0) (Utilities|Struct|MakeGeometryScriptMeshReadLOD "MaxAvailable") (Utilities|Struct|MakeGeometryScriptMeshReadLOD "SourceModel" _index)) :bUseSectionMaterials true)
      (:Success
        (GeometryScript|MeshEdits|AppendMesh (Variables|Default|GetTarget%20Mesh) _returnvalue_11 (Math|Transform|MakeTransform (Math|Vector|MakeVector 0.0 (* _index 256.0))) false "(CombineMode=EnableAllMatching)")
        (bind _returnvalue_12 (GeometryScript|PolyGroups|EnablePolyGroups _returnvalue_11))
        (GeometryScript|PolyGroups|ComputePolyGroupsFromAngleThreshold :TargetMesh _returnvalue_12 :CreaseAngle (Variables|Default|GetMinAngle) :MinGroupSize 1)
        (bind _returnvalue_13 (GeometryScript|Utility|CreateDynamicMeshPool))
        (bind _returnvalue_14 (DynamicMeshActor|AllocateComputeMesh))
        (bind _output_get_2 (Variables|Default|SetTempMeshCurrent _returnvalue_14))
        (Utilities|Array|Add (Variables|Default|GetTempMesh_LODs) _output_get_2)
        (bind _componentmeshes (GeometryScript|Decomposition|SplitMeshByPolyGroups :TargetMesh (Variables|Default|GetCurrent%20Mesh%20LOD) :MeshPool _returnvalue_13))
        (for (_array_element _array_index) _componentmeshes
          (bind (_uvlist _bisvaliduvset) (GeometryScript|UVs|GetMeshPerVertexUVs _array_element 1))
          (bind _uvarray (GeometryScript|ListUtils|ConvertUVListtoArray _uvlist))
          (if _bisvaliduvset
            (Development|PrintString (Utilities|String|ToString(Boolean) (Math|Boolean|ORBoolean (== (Math|Float|Ceil (Math|Vector2D|BreakVector2D (Utilities|Array|Get(aref) _uvarray))) _returnvalue_3) (== (Math|Float|Ceil (Math|Vector2D|BreakVector2D (Utilities|Array|Get(aref) _uvarray))) _returnvalue_8) (== (Math|Float|Ceil (Math|Vector2D|BreakVector2D (Utilities|Array|Get(aref) _uvarray))) _returnvalue_9) (== (Math|Float|Ceil (Math|Vector2D|BreakVector2D (Utilities|Array|Get(aref) _uvarray))) _returnvalue_7))))
            (GeometryScript|VertexColor|SetMeshConstantVertexColor :TargetMesh _array_element :Color (Math|Color|SelectColor (Math|Color|MakeColor (* (/ _returnvalue (Utilities|Array|Length _componentmeshes)) _array_index) (* (/ _returnvalue (Utilities|Array|Length _componentmeshes)) _array_index) (* (/ _returnvalue (Utilities|Array|Length _componentmeshes)) _array_index)) "(R=1.000000,G=1.000000,B=1.000000,A=1.000000)" (Math|Boolean|ORBoolean (== (Math|Float|Ceil (Math|Vector2D|BreakVector2D (Utilities|Array|Get(aref) _uvarray))) _returnvalue_3) (== (Math|Float|Ceil (Math|Vector2D|BreakVector2D (Utilities|Array|Get(aref) _uvarray))) _returnvalue_8) (== (Math|Float|Ceil (Math|Vector2D|BreakVector2D (Utilities|Array|Get(aref) _uvarray))) _returnvalue_9) (== (Math|Float|Ceil (Math|Vector2D|BreakVector2D (Utilities|Array|Get(aref) _uvarray))) _returnvalue_7))) :bClearExisting true)
            (else
              (GeometryScript|VertexColor|SetMeshConstantVertexColor _array_element (Math|Color|MakeColor (* (/ _returnvalue (Utilities|Array|Length _componentmeshes)) _array_index) (* (/ _returnvalue (Utilities|Array|Length _componentmeshes)) _array_index) (* (/ _returnvalue (Utilities|Array|Length _componentmeshes)) _array_index)))))
          (bind _returnvalue_15 (GeometryScript|Repair|WeldMeshEdges _array_element (Utilities|Struct|MakeGeometryScriptWeldEdgesOptions 0.0001)))
          (GeometryScript|MeshEdits|AppendMesh (Variables|Default|GetTempMeshCurrent) _returnvalue_15 (Math|Transform|MakeTransform "0, 0, 0") false "(CombineMode=EnableAllMatching)"))
        (GeometryScript|MeshEdits|AppendMesh (Variables|Default|GetTarget%20Mesh) (Variables|Default|GetTempMeshCurrent) (Math|Transform|MakeTransform (Math|Vector|MakeVector 4e+02 (* _index 256.0))) false "(CombineMode=EnableAllMatching)")))))

(fn Bake
  (for (_array_element _array_index) (Variables|Default|GetTempMesh_LODs)
    (GeometryScript|StaticMesh|CopyMeshtoStaticMesh :FromDynamicMesh _array_element :ToStaticMeshAsset (Variables|Default|GetStaticmesh) :TargetLOD (select (== _array_index 0) (Utilities|Struct|MakeGeometryScriptMeshWriteLOD true) (Utilities|Struct|MakeGeometryScriptMeshWriteLOD false _array_index)) :bUseSectionMaterials true)))

(fn ConstructionScript)
