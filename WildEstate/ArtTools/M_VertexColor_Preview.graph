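# M_VertexColor_Preview is a Default Lit material in the art tools folder, named for previewing mesh vertex colors.
# Its graph holds only a Vertex Color node and a Constant of one, with no links shown between them or to any material output.
# Automatic setting of usage flags in the editor is turned off.
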Material /WildEstate/ArtTools/M_VertexColor_Preview
graph {
 properties {bAutomaticallySetUsageInEditor=false, usedShadingModels=Default Lit}
 nodes [
  Constant() {r=1}
  VertexColor()
 ]
}
